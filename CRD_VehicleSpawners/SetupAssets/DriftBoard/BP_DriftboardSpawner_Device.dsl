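; Driftboard spawner device that applies its audio toggle to the driftboards it spawns.
; On spawn, it copies its bUsesAudio setting onto the Jackal vehicle's bUseAudio.
; Outside preview, BeginPlay binds toy option load and change events to one handler, which on authority re-applies the setting to the cached, already spawned vehicle.
; This keeps sounds such as the death sound correctly played or muted.
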(event EventBeginPlay
  (bind _toyoptionscomponent (Variables|Default|GetToyOptionsComponent))
  ;[1 1 1] Begin Play.  Bind On Properties Changed and On Properties Loaded
  (if (not (Building|IsPreviewBuildingActor))
    (OptionData|BindEventtoOnPropertiesLoaded _toyoptionscomponent (EventDispatchers|CreateEvent :!event Custom|OnPropertiesChangedOrLoaded))
    (OptionData|BindEventtoOnAnyPropertyChanged _toyoptionscomponent (EventDispatchers|CreateEvent :!event Custom|OnPropertiesChangedOrLoaded))))

(event EventOnVehicleSpawned
  (params (Spawned%20Vehicle :type Fort%20Athena%20Vehicle%20Object%20Reference))
  ;[0.8566 0.2885 1] Set whether newly spawned driftboard uses audio
  (|CallParentFunction Spawned%20Vehicle)
  (bind _asjackal_vehicle_athena (Utilities|Casting|CastToJackalVehicle_Athena Spawned%20Vehicle))
  (Variables|/Game/Athena/DrivableVehicles/JackalVehicle_Athena.JackalVehicle_Athena_C|SetbUseAudio (Variables|Default|GetbUsesAudio) _asjackal_vehicle_athena))

(event Custom|OnPropertiesChangedOrLoaded
  (bind _cachedvehicleinterface (Variables|Default|GetCachedVehicleInterface))
  (bind _returnvalue (Utilities|ToObject(Interface) _cachedvehicleinterface))
  ;[0.7222 1 0.3846] Change the current driftboard UseAudio so that UseAudio is applied to an already spawned driftboard and OnDeathSound is played/muted properly
  (Networking|SwitchHasAuthority
    (:Authority
      (Utilities|IsValid _returnvalue
        (:Is%20Valid
          (bind _asjackal_vehicle_athena (Utilities|Casting|CastToJackalVehicle_Athena _returnvalue))
          (Variables|/Game/Athena/DrivableVehicles/JackalVehicle_Athena.JackalVehicle_Athena_C|SetbUseAudio (Variables|Default|GetbUsesAudio) _asjackal_vehicle_athena))) ; This check is done in cast the cached vehicle is pending kill
    )))

(fn ConstructionScript
  (|CallParentFunction))
